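# Nanite-enabled asphalt material for ground meshes in the Apollo Abducted set that blends them into the landscape.
# Vertex red mixes the shared terrain asphalt function with edges built from detail textures and baked AO and normals.
# Height above the landscape RVT height drives separate normal and height-lerped surface blends, with vertex normal Z weighting RVT color.
# A static switch, forward and mobile paths, and material proxies use the plain non-RVT result.
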Material /Game/Environments/Apollo/Sets/Abducted/Props/Materials/M_GroundMeshes_Asphalt
graph {
 properties {
  bTangentSpaceNormal=false
  bUseMaterialAttributes=true
  bUsedWithNanite=true
  pixelDepthOffsetMode=PDOM_Legacy
  usedShadingModels=Default Lit
 }
 nodes [
  n1 = Constant() {r=0.6}
  n2 = TextureCoordinate()
  n3 = AppendVector(
   A=Multiply(A=ComponentMask(None=@n2) {r=true}, B=ScalarParameter() {defaultValue=1, group=Edges, parameterName=Edge Normal Tile})
   B=ComponentMask(None=@n2) {g=true}
  )
  n4 = TextureSampleParameter2D(UVs=@n3) {
   group=Edges
   parameterName=Baked AO
   samplerSource=SSM_Wrap_WorldGroupSettings
   texture=/Game/Environments/Apollo/Sets/Abducted/Props/Textures/T_Abducted_Terrain_Edges_D
  }
  n5 = TextureCoordinate() {uTiling=2, vTiling=0.3}
  n6 = TextureSample(UVs=@n5) {samplerSource=SSM_Wrap_WorldGroupSettings, texture=/Game/Athena/Apollo/Environments/Landscape/Textures/Roads/T_AsphaltDetail1_D}
  n7 = TextureSample(UVs=@n5) {
   samplerSource=SSM_Wrap_WorldGroupSettings
   samplerType=SAMPLERTYPE_Masks
   texture=/Game/Athena/Apollo/Environments/Landscape/Textures/Roads/T_AsphaltDetail1_S
  }
  n8 = TextureSample(UVs=@n5) {samplerType=SAMPLERTYPE_Normal, texture=/Game/Athena/Apollo/Environments/Landscape/Textures/Roads/T_AsphaltDetail1_N}
  n9 = TextureSampleParameter2D(UVs=@n3) {
   group=Edges
   parameterName=Baked Normal
   samplerSource=SSM_Wrap_WorldGroupSettings
   samplerType=SAMPLERTYPE_Normal
   texture=/Game/Environments/Apollo/Sets/Abducted/Props/Textures/T_Abducted_Terrain_Edges_N
  }
  n10 = Multiply(A=@n9.RGB, B=Constant3Vector() {constant=<rgba 1 1 0 0>})
  n11 = SetMaterialAttributes!(
   MaterialAttributes
   Base Color=Multiply(A=@n1, B=Multiply(A=@n4.RGB, B=@n6.RGB))
   Metallic=@n7.G
   Specular=Multiply(A=@n7.R, B=@n1)
   Roughness=Multiply(A=@n7.B) {constB=1.5}
   Normal=Transform(None=Add(A=@n8.RGB, B=@n10))
  ) {attributeSetTypes [
    #1
    #2
    #3
    #4
    #5
   ]}
  n12 = BlendMaterialAttributes(
   A=MaterialFunctionCall!(TilingSize) {desc=Asphalt from terrain, materialFunction=/Game/Athena/Apollo/Environments/Landscape/Materials/MaterialFunctions/MF_ApolloAsphalt}.ForStaticmesh
   B=@n11
   Alpha=VertexColor() {desc=Masking Edges from Top/Bottom}.R
  )
  n13 = RuntimeVirtualTextureSampleParameter() {materialType=BaseColor_Normal_Specular_YCoCg, parameterName=RVT, virtualTexture=/Game/Athena/Apollo/Maps/Landscape/VT/RVT_Apollo_Landscape}
  n14 = RuntimeVirtualTextureSampleParameter() {materialType=WorldHeight, parameterName=RVTHeight, virtualTexture=/Game/Athena/Apollo/Maps/Landscape/VT/RVT_Apollo_Landscape_Height}
  n15 = Subtract(A=ComponentMask(None=WorldPosition().XYZ) {b=true}, B=@n14.WorldHeight)
  n16 = NamedRerouteDeclaration(None=@n15) {name=RVT Height, nodeColor=<rgba 0 0.1094 1 1>}
  n17 = ScalarParameter() {defaultValue=25, group=RVT, parameterName=Blend RVT Normals}
  n18 = BlendMaterialAttributes(
   A=SetMaterialAttributes!(MaterialAttributes=@n12, Normal=@n13.Normal) {attributeSetTypes [#5]}
   B=@n12
   Alpha=Saturate(None=Divide(A=@n16, B=@n17))
  ) {vertexAttributeBlendType=UseA}
  n19 = GetMaterialAttributes(None=@n12) {attributeGetTypes [#1, #5]}
  n20 = VertexNormalWS()
  n21 = SetMaterialAttributes!(
   MaterialAttributes=@n18
   Base Color=LinearInterpolate(A=@n19.Base Color, B=@n13.BaseColor, Alpha=ComponentMask(None=@n20) {b=true})
   Specular=@n13.Specular
   Roughness=@n13.Specular
  ) {attributeSetTypes [#1, #3, #4]}
  n22 = Saturate(None=Divide(A=@n15, B=ScalarParameter() {defaultValue=55, group=RVT, parameterName=Blend RVT}))
  n23 = LinearInterpolate(A=Constant() {r=-1.5}, B=Constant() {r=35}, Alpha=ComponentMask(None=@n19.Base Color) {r=true})
  n24 = Min(
   A=@n22
   B=MaterialFunctionCall(Transition Phase=@n22, Height Texture=Saturate(None=@n23), Contrast=Constant() {r=0.5}) {materialFunction=&HeightLerp}.Alpha
  )
  n25 = MaterialFunctionCall(
   NoRVT=@n12
   RVT=MaterialProxyReplace(Realtime=BlendMaterialAttributes(A=@n21, B=@n18, Alpha=@n24) {vertexAttributeBlendType=UseA}, MaterialProxy=@n12)
  ) {materialFunction=&FN_RVT_Switch_Material}
  StaticSwitchParameter(True=@n12, False=ShadingPathSwitch(Default=@n25.Result, Deferred=@n25.Result, Forward=@n12, Mobile=@n12)) {
   desc="If for some reason you want to use this material but hate to use RVT ( or you'r Bryce L) Will automatically convert details to use up vector instead distance from the ground "
   group=RVT
   parameterName=Skip RVT
   sortPriority=-4
  }
 ]
}
